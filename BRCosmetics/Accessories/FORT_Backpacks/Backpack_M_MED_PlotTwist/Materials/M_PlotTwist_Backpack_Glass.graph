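# Translucent, lit glass material for the PlotTwist backpack cosmetic, with refraction and screen-space reflections.
# Tiled noise from pre-skinned position plus a baked normal feeds a Fresnel that blends inner and outer opacity, with optional texture overrides.
# A static switch layers an alpha-masked diffuse texture with its own surface values before TechArtOverrides and CreativeMoveTool functions.
# The cubemap reflection and sun-highlight chain ends unconnected, so it does not reach the output.
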Material /BRCosmetics/Accessories/FORT_Backpacks/Backpack_M_MED_PlotTwist/Materials/M_PlotTwist_Backpack_Glass
graph {
 properties {
  bAutomaticallySetUsageInEditor=false
  bCastDynamicShadowAsMasked=true
  bEnableResponsiveAA=true
  bScreenSpaceReflections=true
  bUseMaterialAttributes=true
  bUsedWithMeshDeformer=true
  bUsedWithMorphTargets=true
  bUsedWithSkeletalMesh=true
  blendMode=BLEND_Translucent
  ditherOpacityMask=true
  pixelDepthOffsetMode=PDOM_Legacy
  refractionMethod=RM_IndexOfRefraction
  translucencyDirectionalLightingIntensity=5
  translucencyLightingMode=TLM_SurfacePerPixelLighting
  usedShadingModels=Default Lit
 }
 nodes [
  n1 = TextureSampleParameterCube!(UVs=ReflectionVectorWS() {bNormalizeCustomWorldNormal=true}, Bias=Constant() {r=3}, Apply View MipBias) {
   mipValueMode=TMVM_MipBias
   parameterName=CubeMap
   samplerSource=SSM_Wrap_WorldGroupSettings
   samplerType=SAMPLERTYPE_LinearColor
   texture=/Game/Packages/Fortress_Sky/TexturesHDR/T_AthenaSkylight
  }
  n2 = Multiply(A=@n1.RGB, B=VectorParameter() {parameterName=Cube Map Multiplier, defaultValue=<rgba 0.3 0.3 0.3 0>}.RGB) {constB=1.5}
  n3 = ComponentMask(None=Divide(A=PreSkinnedPosition(), B=ObjectBounds())) {b=true, g=true}
  n4 = ScalarParameter() {defaultValue=0.2, parameterName=WaterTiling}
  n5 = TextureSampleParameter2D(UVs=VertexInterpolator(VS=Multiply(A=@n3, B=@n4)).PS) {
   parameterName=Noise Normal
   samplerSource=SSM_Wrap_WorldGroupSettings
   samplerType=SAMPLERTYPE_Normal
   texture=/Game/Global/Textures/Default/Noise/T_Noise_Normal_01
  }
  n6 = VectorParameter() {parameterName=Normal Intensity}
  n7 = TextureSampleParameter2D() {
   group=Mask Texture
   parameterName=BakedNormal
   samplerSource=SSM_Wrap_WorldGroupSettings
   samplerType=SAMPLERTYPE_Normal
   texture=/Engine/EngineMaterials/FlatNormal
  }
  n8 = Normalize(VectorInput=Add(A=Multiply(A=@n5.RGB, B=@n6.RGB), B=@n7.RGB))
  n9 = Fresnel(ExponentIn=ScalarParameter() {defaultValue=1, parameterName=Fresnel Exponent}, Normal=Transform(None=@n8))
  n10 = Multiply(
   A=LinearInterpolate(A=@n2, B=Multiply(A=@n2) {constB=0.05}, Alpha=@n9)
   B=ComponentMask(None=MaterialFunctionCall() {materialFunction=/Engine/Functions/GetPostProcessFunctions/GetAmbientCubemapTint}.AmbientCubemapTint) {b=true, g=true, r=true}
  )
  n11 = CollectionParameter() {
   collection=&FortniteMaterialParameters
   parameterName=SunAndMoonModelDirectionalVector
  }
  n12 = MaterialFunctionCall(Float3=@n11) {materialFunction=&BreakOutFloat3Components}
  n13 = MaterialFunctionCall(X=@n12.R, Y=@n12.R, Z=Constant() {r=1}) {materialFunction=&MakeFloat3}
  n14 = ReflectionVectorWS()
  n15 = Clamp(None=Multiply(A=DotProduct(A=@n13.Result, B=@n14), B=ScalarParameter() {defaultValue=0.8833, parameterName=Spec Pre-Multiplier}))
  n16 = ScalarParameter() {defaultValue=4, parameterName=Spec Highlight Power}
  Add(
   A=Desaturation(None=Multiply(A=@n10) {constB=0.5}, Fraction=Constant() {r=0.4})
   B=Multiply(A=Power(Base=@n15, Exp=@n16) {constExponent=20}, B=ScalarParameter() {defaultValue=3, parameterName=Spec Intense})
  )
  n17 = ScalarParameter() {defaultValue=0.1, parameterName=Roughness}
  LinearInterpolate(Alpha=@n17)
  n18 = StaticSwitchParameter(
   True=VectorParameter() {parameterName=Diffuse Color}.RGB
   False=TextureSampleParameter2D() {parameterName=Diffuse Texture, texture=/Game/Global/Textures/Default/Blanks/T_BlackUserInterface}.RGB
  ) {parameterName=UseDiffuseFlatColor}
  n19 = TextureSampleParameter2D() {parameterName=Input S, texture=&T_BlackUserInterface}
  n20 = Multiply(A=@n9, B=ScalarParameter() {defaultValue=1, parameterName=Glass thickness})
  n21 = LinearInterpolate(
   A=LinearInterpolate(
    A=ScalarParameter() {parameterName=Fresnel Inner Transparency}
    B=ScalarParameter() {defaultValue=0.85, parameterName=Fresnel Outer Transparency}
    Alpha=Clamp(None=@n20)
   )
   B=TextureSampleParameter2D() {parameterName=Input FX, texture=&T_BlackUserInterface}.B
   Alpha=ScalarParameter() {parameterName=Use Custom Opacity}
  )
  n22 = MakeMaterialAttributes!(
   BaseColor=@n18
   Metallic=LinearInterpolate(A=ScalarParameter() {defaultValue=0.1, parameterName=Metallic}, B=@n19.G, Alpha=ScalarParameter() {parameterName=Use Custom Metallic})
   Specular=LinearInterpolate(A=ScalarParameter() {defaultValue=1, parameterName=Specular}, B=@n19.R, Alpha=ScalarParameter() {parameterName=Use Custom Specular})
   Roughness=LinearInterpolate(A=@n17, B=@n19.B, Alpha=ScalarParameter() {parameterName=Use Custom Roughness})
   Anisotropy
   EmissiveColor=Multiply(A=@n18, B=ScalarParameter() {parameterName=EmissiveMult})
   Opacity=@n21
   OpacityMask=@n21
   Normal=@n8
   Tangent
   WorldPositionOffset
   SubsurfaceColor
   ClearCoat
   ClearCoatRoughness
   AmbientOcclusion
   Refraction=ScalarParameter() {defaultValue=1, parameterName=Refraction, sliderMax=2}
   CustomizedUVs_0
   CustomizedUVs_1
   CustomizedUVs_2
   CustomizedUVs_3
   CustomizedUVs_4
   CustomizedUVs_5
   CustomizedUVs_6
   CustomizedUVs_7
   PixelDepthOffset
   ShadingModel
   Displacement
  )
  n23 = GetMaterialAttributes(None=@n22) {attributeGetTypes [
    #1
    #2
    #3
    #4
    #5
    #6
   ]}
  n24 = TextureSampleParameter2D(UVs=TextureCoordinate()) {group=Mask Texture, parameterName=Diffuse Texture with Alpha Mask, sortPriority=10, texture=&T_BlacksRGB}
  n25 = Multiply(A=@n24.A, B=ScalarParameter() {defaultValue=1, group=Mask Texture, parameterName=Alpha Channel Mask Opacity})
  n26 = LinearInterpolate(A=@n23.Opacity, Alpha=@n25)
  n27 = ScalarParameter() {group=Mask Texture, parameterName=Masked Area Emissive Mult, sortPriority=4}
  n28 = SetMaterialAttributes!(
   MaterialAttributes=@n23.MaterialAttributes
   Base Color=LinearInterpolate(A=@n23.Base Color, B=@n24.RGB, Alpha=@n25)
   Metallic=LinearInterpolate(A=@n23.Metallic, B=ScalarParameter() {group=Mask Texture, parameterName=Masked Area Metallic, sortPriority=2}, Alpha=@n25)
   Specular=LinearInterpolate(A=@n23.Specular, B=ScalarParameter() {defaultValue=0.5, group=Mask Texture, parameterName=Masked Area Specular, sortPriority=3}, Alpha=@n25)
   Roughness=LinearInterpolate(
    A=@n23.Roughness
    B=ScalarParameter() {defaultValue=0.4, group=Mask Texture, parameterName=Masked Area Roughness, sortPriority=4}
    Alpha=@n25
   )
   Opacity=@n26
   Opacity Mask=@n26
   Emissive Color=LinearInterpolate(A=@n23.Emissive Color, B=Multiply(A=@n24.RGB, B=@n27), Alpha=@n25)
  ) {attributeSetTypes [
    #1
    #2
    #3
    #4
    #5
    #7
    #6
   ]}
  MaterialFunctionCall(MA In=MaterialFunctionCall(In=StaticSwitchParameter(True=@n28, False=@n22) {group=Mask Texture, parameterName=Use Diffuse Texture with Alpha Mask, sortPriority=-50}) {materialFunction=&MF_TechArtOverrides}.Result) {materialFunction=&MF_CreativeMoveTool}
  Multiply()
 ]
}
